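# Physics asset for the F_MED_Boredom backpack cosmetic, defining collision and joints for its side piece and four wheel bones.
# A two-box pack body is the hub, with one box body for each dyn_ bone constrained directly to it.
# The wheel joints lock swing1 and twist, leaving swing2 as the only rotational axis, while the side piece has no locked rotation axis.
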PhysicsAsset /BRCosmetics/Accessories/FORT_Backpacks/Backpack_F_MED_Boredom/Meshes/F_MED_Boredom_Pack_Physics
bodies {
 dyn_side [{
   shapeType=Box
   extentX=1.01
   extentY=4.01
   extentZ=8.01
   shapeName=dyn_side_box
   center=<5.6951 -0.2153 -0.2025>
   rotation=<rot -0.7218 -90 -19.8257>
  }]
 dyn_wheel_down_1 [{
   shapeType=Box
   extentX=2.339
   extentY=5.5793
   extentZ=5.082
   shapeName=dyn_wheel_down_1_box
   center=<-0.4038 1.1124 0.0244>
   rotation=<rot -0.0657 -90 0.0089>
  }]
 dyn_wheel_down_2 [{
   shapeType=Box
   extentX=2.3385
   extentY=5.5793
   extentZ=5.0818
   shapeName=dyn_wheel_down_2_box
   center=<-0.4038 -0.9736 0.0244>
   rotation=<rot 0.0612 -90 0.0088>
  }]
 dyn_wheel_up_1 [{
   shapeType=Box
   extentX=2.3438
   extentY=5.5793
   extentZ=5.0842
   shapeName=dyn_wheel_up_1_box
   center=<-0.3601 1.1163 0.0182>
   rotation=<rot -0.1205 -90 -0.0231>
  }]
 dyn_wheel_up_2 [{
   shapeType=Box
   extentX=2.3425
   extentY=5.5793
   extentZ=5.0836
   shapeName=dyn_wheel_up_2_box
   center=<-0.3601 -0.9697 0.0182>
   rotation=<rot 0.1058 -90 -0.0214>
  }]
 pack [
  {
   shapeType=Box
   extentX=2
   extentY=2
   extentZ=24.7696
   shapeName=pack_box
   center=<3.7542 -0.3224 -4.7857>
   rotation=<rot 20.2674 -125.2045 26.1242>
  }
  {shapeType=Box, extentX=8.5, extentY=2.5, extentZ=20.5, shapeName=pack_box, center=<0.9515 -0.551 -1.3932>, rotation=<rot -32.6471 0 -0.0001>}
 ]
}
constraints [
 {bone1Name=dyn_side, bone2Name=pack, swing1LimitDegrees=30, swing2LimitDegrees=50, twistLimitDegrees=10}
 {
  bone1Name=dyn_wheel_up_2
  bone2Name=pack
  swing1LimitDegrees=45
  swing1Motion=Locked
  swing2LimitDegrees=75
  twistLimitDegrees=45
  twistMotion=Locked
 }
 {
  bone1Name=dyn_wheel_down_1
  bone2Name=pack
  swing1LimitDegrees=45
  swing1Motion=Locked
  swing2LimitDegrees=55
  twistLimitDegrees=45
  twistMotion=Locked
 }
 {
  bone1Name=dyn_wheel_down_2
  bone2Name=pack
  swing1LimitDegrees=45
  swing1Motion=Locked
  swing2LimitDegrees=35
  twistLimitDegrees=45
  twistMotion=Locked
 }
 {
  bone1Name=dyn_wheel_up_1
  bone2Name=pack
  swing1LimitDegrees=45
  swing1Motion=Locked
  swing2LimitDegrees=45
  twistLimitDegrees=45
  twistMotion=Locked
 }
]
