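# Animation Blueprint for the F_MED_Boredom backpack cosmetic, adding physics motion and pose offsets to the pack.
# A RigidBody node simulates the pack's physics asset in base-bone space, with world gravity driven by GravityOverride.
# Each update, a deprecated gravity-override state function sets that vector from per-state parameters.
# Additive bone-space transforms then adjust the pack bone, including blended offsets for skydive floating and open parachute.
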AnimBlueprint /BRCosmetics/Accessories/FORT_Backpacks/Backpack_F_MED_Boredom/Meshes/F_MED_Boredom_Pack_AnimBP
target_skeleton=&/F_MED_Boredom_Pack_Skeleton
graphs [
 AnimGraph {nodes [
   n1 = RigidBody!(ComponentPose=Local To Component(LocalPose=Input Pose()), Alpha, OverrideWorldGravity:Vector=Get GravityOverride(), ExternalForce) -> (Pose) {
    bClampLinearTranslationLimitToRefPose=true
    bOverrideWorldGravity=true
    overridePhysicsAsset=&/F_MED_Boredom_Pack_Physics
    simulationSpace=BaseBoneSpace
    alphaBoolBlend {blendOption=Linear}
    componentLinearAccScale=<0.5 0.5 0.5>
    componentLinearVelScale=<0.5 0.5 0.5>
    overrideWorldGravity {z=-980}
   }
   n2 = "Transform (Modify) Bone - Bone: pack"!(ComponentPose=@n1, bAlphaBoolEnabled:bool, Translation="0.718,-1.377,0.069", Rotation="4.124,1.936,2.823", Scale) -> (Pose) {
    alphaInputType=Bool
    rotationMode=BMM_Additive
    rotationSpace=BCS_BoneSpace
    scaleMode=BMM_Additive
    scaleSpace=BCS_BoneSpace
    translationMode=BMM_Additive
    translationSpace=BCS_BoneSpace
    alphaBoolBlend {blendInTime=0.1, blendOption=Linear, blendOutTime=1}
    boneToModify {boneName=pack}
    rotation=<rot 4.124 1.936 2.823>
    translation=<0.718 -1.377 0.069>
   }
   n3 = "Transform (Modify) Bone - Bone: pack"(ComponentPose=@n2, bAlphaBoolEnabled=Get bIsSkydiveFloating(), Translation="0.665,-6.899,1.053", Rotation="-7.053,8.430,-0.830") {
    alphaInputType=Bool
    rotationMode=BMM_Additive
    rotationSpace=BCS_BoneSpace
    scaleMode=BMM_Additive
    scaleSpace=BCS_BoneSpace
    translationMode=BMM_Additive
    translationSpace=BCS_BoneSpace
    alphaBoolBlend {blendInTime=0.1, blendOption=Linear, blendOutTime=1}
    boneToModify {boneName=pack}
    rotation=<rot -7.053 8.43 -0.83>
    translation=<0.665 -6.899 1.053>
   }
   n4 = "Transform (Modify) Bone - Bone: pack"(ComponentPose=@n3, bAlphaBoolEnabled=Get bIsParachuteOpened(), Translation="0.000,-1.927,-0.050", Rotation="-0.000,0.000,12.988") {
    alphaInputType=Bool
    rotationMode=BMM_Additive
    rotationSpace=BCS_BoneSpace
    scaleMode=BMM_Additive
    scaleSpace=BCS_BoneSpace
    translationMode=BMM_Additive
    translationSpace=BCS_BoneSpace
    alphaBoolBlend {blendInTime=0.1, blendOption=Linear, blendOutTime=0.3}
    boneToModify {boneName=pack}
    rotation {roll=12.988}
    translation {y=-1.927, z=-0.05}
   }
   Output Pose(Result=Component To Local(ComponentPose=@n4)) {layerGroup=None}
  ]}
 EventGraph {nodes [
   Event BlueprintUpdateAnimation()
   > n1 = Gravity Override State Function DEPRECATED(
    CharacterPartAnimInstance=Self-Reference()
    GameTimeInSeconds=GetGameTimeInSeconds()
    GravityOverrideParams=Get Gravity override Params() -> (Gravity override Params:Gravity Override Params Struct Structure)
   )
   > Set GravityOverride(GravityOverride=@n1)
   TryGetPawnOwner()
  ]}
]
variables [
 Gravity Override Params {
  type=GravityOverrideParamsStruct
  default {
   dBNO_Carried {gravityOverride=<-980 500 -980>}
   riding_Moving {gravityOverride=<0 -450 -980>}
   tactical_Sprint {gravityOverride=<0 -450 -980>}
  }
 }
 GravityOverride {type=Vector, default=<0 0 -980>}
]
